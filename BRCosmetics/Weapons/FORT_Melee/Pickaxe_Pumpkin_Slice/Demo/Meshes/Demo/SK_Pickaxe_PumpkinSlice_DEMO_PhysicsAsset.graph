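# Physics asset for the Pumpkin Slice demo pickaxe skeletal mesh, a two-handed melee weapon cosmetic.
# A single capsule body on the Melee_TwoHanded_Handle bone serves as the weapon's collision shape.
# It has no constraints, so there are no jointed or separately simulated parts.
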PhysicsAsset /BRCosmetics/Weapons/FORT_Melee/Pickaxe_Pumpkin_Slice/Demo/Meshes/Demo/SK_Pickaxe_PumpkinSlice_DEMO_PhysicsAsset
bodies {Melee_TwoHanded_Handle [{shapeType=Capsule, length=67.6669, radius=26.7004, center=<0.7077 8.8743 34.3302>, rotation=<rot 5.1106 -90 -0.0151>}]}
constraints []
